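# Physics asset for the PrairieGizmo female medium body's skirt, beads and rope secondary motion.
# Limited-swing capsule chains hang from tiny deform-bone anchors tied to pelvis (back) and thighs (sides), with a left-front panel on the thigh.
# Beads and upper rope mix angle-locked pelvis joints with limited links to each other and the left side skirt.
# Leg, pelvis and spine capsules plus a large below-root sphere provide body shapes, mostly joined by free-angular constraints.
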PhysicsAsset /BRCosmetics/Characters/Player/Female/Medium/Bodies/F_MED_PrairieGizmo/Meshes/F_MED_PrairieGizmo_PhysicsSkirt
bodies {
 calf_l [
  {shapeType=Capsule, length=17.0018, radius=8.7708, shapeName=calf_l_capsule, center=<-13.6327 -2.0199 0>, rotation=<rot 0 95.3936 90>}
  {
   shapeType=Capsule
   length=40.061
   radius=9.9436
   shapeName=calf_l_capsule
   center=<-31.9672 1.2815 0.0171>
   rotation=<rot -0.3859 86.1009 89.9295>
  }
 ]
 calf_r [
  {shapeType=Capsule, length=17.0018, radius=8.7708, shapeName=calf_r_capsule, center=<13.6327 2.0199 0>, rotation=<rot 0 95.3936 -90>}
  {
   shapeType=Capsule
   length=40.061
   radius=9.9436
   shapeName=calf_r_capsule
   center=<31.9672 -1.2815 -0.0171>
   rotation=<rot -0.3859 86.1009 -90.0705>
  }
 ]
 deform_skirt_back_l [{shapeType=Capsule, length=0.505, radius=0.505, shapeName=deform_skirt_back_l_capsule}]
 deform_skirt_back_r [{shapeType=Capsule, length=0.505, radius=0.505, shapeName=deform_skirt_back_r_capsule}]
 deform_skirt_side_l [{shapeType=Capsule, length=0.505, radius=0.505, shapeName=deform_skirt_side_l_capsule}]
 deform_skirt_side_r [{shapeType=Capsule, length=0.505, radius=0.505, shapeName=deform_skirt_side_r_capsule}]
 dyn_beads_04 [{shapeType=Capsule, length=16.1109, radius=7.8891, shapeName=dyn_beads_04_capsule, center=<0 0 7.0055>, rotation=<rot 90 90 0>}]
 dyn_beads_05 [{shapeType=Capsule, length=16.1109, radius=7.8891, shapeName=dyn_beads_04_capsule, center=<0 0 3.2061>, rotation=<rot 90 90 0>}]
 dyn_beads_06 [{shapeType=Capsule, length=16.1109, radius=7.8891, shapeName=dyn_beads_04_capsule, center=<0 0 10.1948>, rotation=<rot 90 90 0>}]
 dyn_rope_upper_01 [{shapeType=Capsule, length=16.1109, radius=7.8891, shapeName=dyn_beads_04_capsule, center=<0 0 6.9137>, rotation=<rot 90 90 0>}]
 dyn_rope_upper_02 [{shapeType=Capsule, length=16.1109, radius=7.8891, shapeName=dyn_beads_04_capsule, center=<0 0 5.2868>, rotation=<rot 90 90 0>}]
 dyn_rope_upper_03 [{
   shapeType=Capsule
   length=16.1109
   radius=7.8891
   shapeName=dyn_beads_04_capsule
   center=<0 -0.0305 4.1059>
   rotation=<rot 85.8175 90.0013 0.0012>
  }]
 dyn_skirt_back_01_l [{
   shapeType=Capsule
   length=16.4872
   radius=10.8486
   shapeName=dyn_skirt_back_01_l_capsule
   center=<-8.5613 8.5 -3.3384>
   rotation=<rot 90 -90 0>
  }]
 dyn_skirt_back_01_r [{
   shapeType=Capsule
   length=16.4872
   radius=10.8493
   shapeName=dyn_skirt_back_01_r_capsule
   center=<8.5613 -8.5 3.3384>
   rotation=<rot 90 -90 0>
  }]
 dyn_skirt_back_02_l [{
   shapeType=Capsule
   length=22.6792
   radius=12.3803
   shapeName=dyn_skirt_back_02_l_capsule
   center=<-12.2754 -1.4243 -7.9991>
   rotation=<rot 90 -90 0>
  }]
 dyn_skirt_back_02_r [{
   shapeType=Capsule
   length=22.6792
   radius=12.3809
   shapeName=dyn_skirt_back_02_r_capsule
   center=<12.2754 -6.0729 7.9991>
   rotation=<rot 90 -90 0>
  }]
 dyn_skirt_back_base_l [{
   shapeType=Capsule
   length=10.2585
   radius=8.5202
   shapeName=dyn_skirt_back_base_l_capsule
   center=<-12.1123 -1.5188 -5.5109>
   rotation=<rot 90 -90 0>
  }]
 dyn_skirt_back_base_r [{
   shapeType=Capsule
   length=10.2585
   radius=8.5202
   shapeName=dyn_skirt_back_base_r_capsule
   center=<12.1123 1.5188 5.5109>
   rotation=<rot 90 90 0>
  }]
 dyn_skirt_front_01_l [{shapeType=Capsule, length=10.4717, radius=7.6169, shapeName=dyn_skirt_side_01_l_capsule, center=<-17.4845 0 -6.4788>, rotation=<rot 90 90 0>}]
 dyn_skirt_side_01_l [{shapeType=Capsule, length=23.2126, radius=7.6169, shapeName=dyn_skirt_side_01_l_capsule, center=<-12 0 -4.6836>, rotation=<rot 90 90 0>}]
 dyn_skirt_side_01_r [{shapeType=Capsule, length=23.2126, radius=7.6169, shapeName=dyn_skirt_side_01_r_capsule, center=<12 0 5.4237>, rotation=<rot 90 -90 0>}]
 dyn_skirt_side_base_l [{
   shapeType=Capsule
   length=15.7769
   radius=8.2231
   shapeName=dyn_skirt_side_base_l_capsule
   center=<-11.5756 0 -6.8088>
   rotation=<rot 90 -90 0>
  }]
 dyn_skirt_side_base_r [{shapeType=Capsule, length=15.7769, radius=8.2231, shapeName=dyn_skirt_side_base_r_capsule, center=<11.5756 0 8.3781>, rotation=<rot 90 90 0>}]
 foot_l [{shapeType=Capsule, length=1.5696, radius=14.4717, shapeName=foot_l_capsule, center=<-2.5446 4.4744 0.8404>, rotation=<rot -5.7371 89.9998 0>}]
 foot_r [{
   shapeType=Capsule
   length=1.5696
   radius=14.4717
   shapeName=foot_r_capsule
   center=<2.5446 -4.4744 -0.8404>
   rotation=<rot -5.7371 89.9998 180>
  }]
 pelvis [{shapeType=Capsule, length=9.9915, radius=12.0131, shapeName=pelvis_capsule, center=<-4.4947 -0.0086 0>, rotation=<rot 90 -0.0001 0>}]
 root [{shapeType=Sphere, radius=130.8217, shapeName=root_sphere, center=<0 -1.0809 -119.3489>}]
 spine_01 [{shapeType=Capsule, length=15.194, radius=9.8299, shapeName=spine_01_capsule, center=<-17.6762 3.7507 0>, rotation=<rot 0 -90 0>}]
 thigh_l [{
   shapeType=Capsule
   length=44
   radius=7.8391
   shapeName=thigh_l_capsule
   center=<-16.7742 0.3308 1.4316>
   rotation=<rot 89.8621 0.0074 0.0078>
  }]
 thigh_r [{
   shapeType=Capsule
   length=34.7424
   radius=7.8391
   shapeName=thigh_r_capsule
   center=<25.4477 -0.3309 -0.9007>
   rotation=<rot 89.8621 0.0074 -179.9922>
  }]
 thigh_twist_01_l [{
   shapeType=Capsule
   length=17.0058
   radius=7.8391
   shapeName=thigh_l_capsule
   center=<19.3234 3.5733 1.3447>
   rotation=<rot 89.8621 0.0074 0.0078>
  }]
}
constraints [
 {bone1Name=dyn_skirt_back_01_r, bone2Name=dyn_skirt_back_base_r, swing1LimitDegrees=15, swing2LimitDegrees=40, twistLimitDegrees=3}
 {bone1Name=dyn_skirt_back_02_r, bone2Name=dyn_skirt_back_01_r, swing1LimitDegrees=35, swing2LimitDegrees=35, twistLimitDegrees=10}
 {bone1Name=dyn_skirt_back_02_l, bone2Name=dyn_skirt_back_01_l, swing1LimitDegrees=35, swing2LimitDegrees=35, twistLimitDegrees=10}
 {bone1Name=dyn_skirt_back_01_l, bone2Name=dyn_skirt_back_base_l, swing1LimitDegrees=20, swing2LimitDegrees=40, twistLimitDegrees=3}
 {bone1Name=dyn_skirt_side_01_l, bone2Name=dyn_skirt_side_base_l, swing1LimitDegrees=25, swing2LimitDegrees=25, twistLimitDegrees=3}
 {bone1Name=dyn_skirt_side_01_r, bone2Name=dyn_skirt_side_base_r, swing1LimitDegrees=25, swing2LimitDegrees=25, twistLimitDegrees=3}
 {bone1Name=dyn_skirt_back_base_l, bone2Name=deform_skirt_back_l, swing1LimitDegrees=10, swing2LimitDegrees=30, twistLimitDegrees=3}
 {bone1Name=dyn_skirt_back_base_r, bone2Name=deform_skirt_back_r, swing1LimitDegrees=10, swing2LimitDegrees=30, twistLimitDegrees=3}
 {bone1Name=dyn_skirt_side_base_l, bone2Name=deform_skirt_side_l, swing1LimitDegrees=1, swing2LimitDegrees=15, twistLimitDegrees=1}
 {bone1Name=dyn_skirt_side_base_r, bone2Name=deform_skirt_side_r, swing1LimitDegrees=1, swing2LimitDegrees=15, twistLimitDegrees=1}
 {
  bone1Name=deform_skirt_side_l
  bone2Name=thigh_l
  swing1LimitDegrees=45
  swing1Motion=Free
  swing2LimitDegrees=45
  swing2Motion=Free
  twistLimitDegrees=45
  twistMotion=Free
 }
 {
  bone1Name=deform_skirt_side_r
  bone2Name=thigh_r
  swing1LimitDegrees=45
  swing1Motion=Free
  swing2LimitDegrees=45
  swing2Motion=Free
  twistLimitDegrees=45
  twistMotion=Free
 }
 {
  bone1Name=thigh_l
  bone2Name=pelvis
  swing1LimitDegrees=45
  swing1Motion=Free
  swing2LimitDegrees=45
  swing2Motion=Free
  twistLimitDegrees=45
  twistMotion=Free
 }
 {
  bone1Name=thigh_r
  bone2Name=pelvis
  swing1LimitDegrees=45
  swing1Motion=Free
  swing2LimitDegrees=45
  swing2Motion=Free
  twistLimitDegrees=45
  twistMotion=Free
 }
 {
  bone1Name=deform_skirt_back_r
  bone2Name=pelvis
  swing1LimitDegrees=45
  swing1Motion=Free
  swing2LimitDegrees=45
  swing2Motion=Free
  twistLimitDegrees=45
  twistMotion=Free
 }
 {
  bone1Name=deform_skirt_back_l
  bone2Name=pelvis
  swing1LimitDegrees=45
  swing1Motion=Free
  swing2LimitDegrees=45
  swing2Motion=Free
  twistLimitDegrees=45
  twistMotion=Free
 }
 {
  bone1Name=calf_r
  bone2Name=thigh_r
  swing1LimitDegrees=45
  swing1Motion=Free
  swing2LimitDegrees=45
  swing2Motion=Free
  twistLimitDegrees=45
  twistMotion=Free
 }
 {
  bone1Name=calf_l
  bone2Name=thigh_l
  swing1LimitDegrees=45
  swing1Motion=Free
  swing2LimitDegrees=45
  swing2Motion=Free
  twistLimitDegrees=45
  twistMotion=Free
 }
 {
  bone1Name=foot_r
  bone2Name=calf_r
  swing1LimitDegrees=45
  swing1Motion=Free
  swing2LimitDegrees=45
  swing2Motion=Free
  twistLimitDegrees=45
  twistMotion=Free
 }
 {
  bone1Name=foot_l
  bone2Name=calf_l
  swing1LimitDegrees=45
  swing1Motion=Free
  swing2LimitDegrees=45
  swing2Motion=Free
  twistLimitDegrees=45
  twistMotion=Free
 }
 {
  bone1Name=pelvis
  bone2Name=root
  swing1LimitDegrees=45
  swing1Motion=Free
  swing2LimitDegrees=45
  swing2Motion=Free
  twistLimitDegrees=45
  twistMotion=Free
 }
 {
  bone1Name=dyn_skirt_back_02_r
  bone2Name=dyn_skirt_back_02_l
  swing1LimitDegrees=45
  swing2LimitDegrees=45
  swing2Motion=Free
  twistLimitDegrees=45
  twistMotion=Free
 }
 {
  bone1Name=thigh_r
  bone2Name=dyn_skirt_back_02_r
  swing1LimitDegrees=25
  swing1Motion=Free
  swing2LimitDegrees=25
  swing2Motion=Free
  twistLimitDegrees=25
  twistMotion=Free
 }
 {
  bone1Name=thigh_l
  bone2Name=dyn_skirt_back_02_l
  swing1LimitDegrees=45
  swing1Motion=Free
  swing2LimitDegrees=25
  swing2Motion=Free
  twistLimitDegrees=45
  twistMotion=Free
 }
 {
  bone1Name=pelvis
  bone2Name=dyn_skirt_side_01_l
  swing1LimitDegrees=45
  swing1Motion=Free
  swing2LimitDegrees=45
  swing2Motion=Free
  twistLimitDegrees=45
  twistMotion=Free
 }
 {
  bone1Name=dyn_beads_05
  bone2Name=pelvis
  swing1LimitDegrees=45
  swing1Motion=Locked
  swing2LimitDegrees=5
  swing2Motion=Locked
  twistLimitDegrees=45
  twistMotion=Locked
 }
 {
  bone1Name=dyn_beads_04
  bone2Name=dyn_skirt_side_base_l
  swing1LimitDegrees=120
  swing2LimitDegrees=5
  swing2Motion=Locked
  twistLimitDegrees=5
 }
 {bone1Name=dyn_beads_04, bone2Name=dyn_beads_05, swing1LimitDegrees=45, swing2LimitDegrees=45, twistLimitDegrees=45}
 {
  bone1Name=spine_01
  bone2Name=pelvis
  swing1LimitDegrees=45
  swing1Motion=Free
  swing2LimitDegrees=45
  swing2Motion=Free
  twistLimitDegrees=45
  twistMotion=Free
 }
 {
  bone1Name=dyn_beads_06
  bone2Name=pelvis
  swing1LimitDegrees=45
  swing1Motion=Locked
  swing2LimitDegrees=5
  swing2Motion=Locked
  twistLimitDegrees=45
  twistMotion=Locked
 }
 {
  bone1Name=dyn_rope_upper_01
  bone2Name=pelvis
  swing1LimitDegrees=45
  swing1Motion=Locked
  swing2LimitDegrees=5
  swing2Motion=Locked
  twistLimitDegrees=45
  twistMotion=Locked
 }
 {
  bone1Name=dyn_rope_upper_03
  bone2Name=pelvis
  swing1LimitDegrees=45
  swing1Motion=Locked
  swing2LimitDegrees=5
  swing2Motion=Locked
  twistLimitDegrees=45
  twistMotion=Locked
 }
 {bone1Name=dyn_rope_upper_02, bone2Name=dyn_rope_upper_03, swing1LimitDegrees=45, swing2LimitDegrees=45, twistLimitDegrees=45}
 {
  bone1Name=dyn_rope_upper_02
  bone2Name=dyn_skirt_side_base_l
  swing1LimitDegrees=120
  swing2LimitDegrees=5
  swing2Motion=Locked
  twistLimitDegrees=5
 }
 {bone1Name=dyn_skirt_front_01_l, bone2Name=thigh_l, swing1LimitDegrees=1, swing2LimitDegrees=30, twistLimitDegrees=45, twistMotion=Locked}
 {
  bone1Name=thigh_twist_01_l
  bone2Name=thigh_l
  swing1LimitDegrees=45
  swing1Motion=Free
  swing2LimitDegrees=45
  swing2Motion=Free
  twistLimitDegrees=45
  twistMotion=Free
 }
]
