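# Physics asset for the PrairieGizmo female medium body, a DBNO variant built around skirt dynamics.
# Back and side skirt chains use tight swing and twist limits, hanging from tiny deform-bone capsules linked to pelvis and thighs.
# Cross constraints tie the back tips to each other and to the thighs.
# Capsules cover pelvis and legs, and a large sphere below root likely acts as a floor proxy.
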PhysicsAsset /BRCosmetics/Characters/Player/Female/Medium/Bodies/F_MED_PrairieGizmo/Meshes/F_MED_PrairieGizmo_Physics_DBNO2
bodies {
 calf_l [
  {shapeType=Capsule, length=17.0018, radius=8.7708, shapeName=calf_l_capsule, center=<-13.6327 -2.0199 0>, rotation=<rot 0 95.3936 90>}
  {
   shapeType=Capsule
   length=34.0906
   radius=9.9436
   shapeName=calf_l_capsule
   center=<-29.9615 1.115 -0.0613>
   rotation=<rot -0.3859 86.1009 89.9295>
  }
 ]
 calf_r [
  {shapeType=Capsule, length=17.0018, radius=8.7708, shapeName=calf_r_capsule, center=<13.6327 2.0199 0>, rotation=<rot 0 95.3936 -90>}
  {
   shapeType=Capsule
   length=34.0906
   radius=9.9436
   shapeName=calf_r_capsule
   center=<29.9615 -1.115 0.0613>
   rotation=<rot -0.3859 86.1009 -90.0705>
  }
 ]
 deform_skirt_back_l [{shapeType=Capsule, length=0.505, radius=0.505, shapeName=deform_skirt_back_l_capsule}]
 deform_skirt_back_r [{shapeType=Capsule, length=0.505, radius=0.505, shapeName=deform_skirt_back_r_capsule}]
 deform_skirt_side_l [{shapeType=Capsule, length=0.505, radius=0.505, shapeName=deform_skirt_side_l_capsule}]
 deform_skirt_side_r [{shapeType=Capsule, length=0.505, radius=0.505, shapeName=deform_skirt_side_r_capsule}]
 dyn_skirt_back_01_l [{
   shapeType=Capsule
   length=16.3469
   radius=10.9889
   shapeName=dyn_skirt_back_01_l_capsule
   center=<-9.9029 -0.3253 -3.2892>
   rotation=<rot 90 -90 0>
  }]
 dyn_skirt_back_01_r [{
   shapeType=Capsule
   length=16.3469
   radius=10.9896
   shapeName=dyn_skirt_back_01_r_capsule
   center=<9.902 0.2964 3.2443>
   rotation=<rot 90 -90 0>
  }]
 dyn_skirt_back_02_l [{
   shapeType=Capsule
   length=22.6792
   radius=12.3803
   shapeName=dyn_skirt_back_02_l_capsule
   center=<-12.359 -1.9923 -11.3014>
   rotation=<rot 90 -90 0>
  }]
 dyn_skirt_back_02_r [{
   shapeType=Capsule
   length=22.6792
   radius=12.3809
   shapeName=dyn_skirt_back_02_r_capsule
   center=<12.3581 -5.5088 11.3022>
   rotation=<rot 90 -90 0>
  }]
 dyn_skirt_back_base_l [{
   shapeType=Capsule
   length=10.2585
   radius=8.5202
   shapeName=dyn_skirt_back_base_l_capsule
   center=<-12.1123 -1.5188 -5.5109>
   rotation=<rot 90 -90 0>
  }]
 dyn_skirt_back_base_r [{
   shapeType=Capsule
   length=10.2585
   radius=8.5202
   shapeName=dyn_skirt_back_base_r_capsule
   center=<12.1123 1.5188 5.5109>
   rotation=<rot 90 90 0>
  }]
 dyn_skirt_side_01_l [{shapeType=Capsule, length=24.9396, radius=4.4475, shapeName=dyn_skirt_side_01_l_capsule, center=<-12 0 -1.632>, rotation=<rot 90 -90 0>}]
 dyn_skirt_side_01_r [{shapeType=Capsule, length=23.2126, radius=7.6169, shapeName=dyn_skirt_side_01_r_capsule, center=<12 0 5.4237>, rotation=<rot 90 -90 0>}]
 dyn_skirt_side_base_l [{
   shapeType=Capsule
   length=17.2418
   radius=6.7582
   shapeName=dyn_skirt_side_base_l_capsule
   center=<-12.3783 -0.0016 -5.1237>
   rotation=<rot 90 -90 0>
  }]
 dyn_skirt_side_base_r [{
   shapeType=Capsule
   length=17.2418
   radius=6.7582
   shapeName=dyn_skirt_side_base_r_capsule
   center=<12.3639 0.0016 4.9279>
   rotation=<rot 90 90 0>
  }]
 foot_l [{shapeType=Capsule, length=0.2459, radius=15.7954, shapeName=foot_l_capsule, center=<3.6283 5.3091 0.9516>, rotation=<rot -5.7371 89.9998 0>}]
 foot_r [{
   shapeType=Capsule
   length=0.2459
   radius=15.7954
   shapeName=foot_r_capsule
   center=<-3.6283 -5.3091 -0.9516>
   rotation=<rot -5.7371 89.9998 180>
  }]
 pelvis [{shapeType=Capsule, length=9.9915, radius=12.0131, shapeName=pelvis_capsule, center=<-1.6873 -0.1093 0>, rotation=<rot 90 -0.0001 0>}]
 root [{shapeType=Sphere, radius=130.8217, shapeName=root_sphere, center=<0 -1.0809 -119.3489>}]
 thigh_l [{
   shapeType=Capsule
   length=31.4458
   radius=6.0353
   shapeName=thigh_l_capsule
   center=<-28.3563 0.0002 0.0682>
   rotation=<rot 89.8621 0.0074 0.0078>
  }]
 thigh_r [{
   shapeType=Capsule
   length=29.642
   radius=7.8391
   shapeName=thigh_r_capsule
   center=<28.3543 -0.3309 -0.9077>
   rotation=<rot 89.8621 0.0074 -179.9922>
  }]
}
constraints [
 {bone1Name=dyn_skirt_back_01_r, bone2Name=dyn_skirt_back_base_r, swing1LimitDegrees=10, swing2LimitDegrees=40, twistLimitDegrees=3}
 {bone1Name=dyn_skirt_back_02_r, bone2Name=dyn_skirt_back_01_r, swing1LimitDegrees=15, swing2LimitDegrees=35, twistLimitDegrees=3}
 {bone1Name=dyn_skirt_back_02_l, bone2Name=dyn_skirt_back_01_l, swing1LimitDegrees=15, swing2LimitDegrees=35, twistLimitDegrees=3}
 {bone1Name=dyn_skirt_back_01_l, bone2Name=dyn_skirt_back_base_l, swing1LimitDegrees=10, swing2LimitDegrees=40, twistLimitDegrees=3}
 {bone1Name=dyn_skirt_side_01_l, bone2Name=dyn_skirt_side_base_l, swing1LimitDegrees=25, swing2LimitDegrees=25, twistLimitDegrees=3}
 {bone1Name=dyn_skirt_side_01_r, bone2Name=dyn_skirt_side_base_r, swing1LimitDegrees=25, swing2LimitDegrees=25, twistLimitDegrees=3}
 {bone1Name=dyn_skirt_back_base_l, bone2Name=deform_skirt_back_l, swing1LimitDegrees=10, swing2LimitDegrees=30, twistLimitDegrees=3}
 {bone1Name=dyn_skirt_back_base_r, bone2Name=deform_skirt_back_r, swing1LimitDegrees=10, swing2LimitDegrees=30, twistLimitDegrees=3}
 {bone1Name=dyn_skirt_side_base_l, bone2Name=deform_skirt_side_l, swing1LimitDegrees=2, swing2LimitDegrees=25, twistLimitDegrees=2}
 {bone1Name=dyn_skirt_side_base_r, bone2Name=deform_skirt_side_r, swing1LimitDegrees=1, swing2LimitDegrees=15, twistLimitDegrees=1}
 {
  bone1Name=deform_skirt_side_l
  bone2Name=thigh_l
  swing1LimitDegrees=45
  swing1Motion=Free
  swing2LimitDegrees=45
  swing2Motion=Free
  twistLimitDegrees=45
  twistMotion=Free
 }
 {
  bone1Name=deform_skirt_side_r
  bone2Name=thigh_r
  swing1LimitDegrees=45
  swing1Motion=Free
  swing2LimitDegrees=45
  swing2Motion=Free
  twistLimitDegrees=45
  twistMotion=Free
 }
 {
  bone1Name=thigh_l
  bone2Name=pelvis
  swing1LimitDegrees=45
  swing1Motion=Free
  swing2LimitDegrees=45
  swing2Motion=Free
  twistLimitDegrees=45
  twistMotion=Free
 }
 {
  bone1Name=thigh_r
  bone2Name=pelvis
  swing1LimitDegrees=45
  swing1Motion=Free
  swing2LimitDegrees=45
  swing2Motion=Free
  twistLimitDegrees=45
  twistMotion=Free
 }
 {
  bone1Name=deform_skirt_back_r
  bone2Name=pelvis
  swing1LimitDegrees=45
  swing1Motion=Free
  swing2LimitDegrees=45
  swing2Motion=Free
  twistLimitDegrees=45
  twistMotion=Free
 }
 {
  bone1Name=deform_skirt_back_l
  bone2Name=pelvis
  swing1LimitDegrees=45
  swing1Motion=Free
  swing2LimitDegrees=45
  swing2Motion=Free
  twistLimitDegrees=45
  twistMotion=Free
 }
 {
  bone1Name=calf_r
  bone2Name=thigh_r
  swing1LimitDegrees=45
  swing1Motion=Free
  swing2LimitDegrees=45
  swing2Motion=Free
  twistLimitDegrees=45
  twistMotion=Free
 }
 {
  bone1Name=calf_l
  bone2Name=thigh_l
  swing1LimitDegrees=45
  swing1Motion=Free
  swing2LimitDegrees=45
  swing2Motion=Free
  twistLimitDegrees=45
  twistMotion=Free
 }
 {
  bone1Name=foot_r
  bone2Name=calf_r
  swing1LimitDegrees=45
  swing1Motion=Free
  swing2LimitDegrees=45
  swing2Motion=Free
  twistLimitDegrees=45
  twistMotion=Free
 }
 {
  bone1Name=foot_l
  bone2Name=calf_l
  swing1LimitDegrees=45
  swing1Motion=Free
  swing2LimitDegrees=45
  swing2Motion=Free
  twistLimitDegrees=45
  twistMotion=Free
 }
 {
  bone1Name=pelvis
  bone2Name=root
  swing1LimitDegrees=45
  swing1Motion=Free
  swing2LimitDegrees=45
  swing2Motion=Free
  twistLimitDegrees=45
  twistMotion=Free
 }
 {
  bone1Name=dyn_skirt_back_02_r
  bone2Name=dyn_skirt_back_02_l
  swing1LimitDegrees=15
  swing2LimitDegrees=45
  swing2Motion=Free
  twistLimitDegrees=45
  twistMotion=Free
 }
 {
  bone1Name=thigh_r
  bone2Name=dyn_skirt_back_02_r
  swing1LimitDegrees=25
  swing1Motion=Free
  swing2LimitDegrees=25
  swing2Motion=Free
  twistLimitDegrees=25
  twistMotion=Free
 }
 {
  bone1Name=thigh_l
  bone2Name=dyn_skirt_back_02_l
  swing1LimitDegrees=45
  swing1Motion=Free
  swing2LimitDegrees=25
  swing2Motion=Free
  twistLimitDegrees=45
  twistMotion=Free
 }
]
