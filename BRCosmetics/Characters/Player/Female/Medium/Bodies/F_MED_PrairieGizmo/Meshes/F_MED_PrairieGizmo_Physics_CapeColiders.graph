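# Physics asset for the female F_MED_PrairieGizmo cosmetic body, named for cape colliders and covering its skirt bones and body.
# Capsules on the pelvis, spine, legs and feet form the body volumes, joined by angularly free constraints.
# Back and side skirt chains use tight swing and twist limits, rooted through tiny deform-bone capsules to pelvis and thighs respectively.
# Extra constraints link the lower back skirt bones to each other and the thighs.
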PhysicsAsset /BRCosmetics/Characters/Player/Female/Medium/Bodies/F_MED_PrairieGizmo/Meshes/F_MED_PrairieGizmo_Physics_CapeColiders
bodies {
 calf_l [
  {shapeType=Capsule, length=17.0018, radius=8.7708, shapeName=calf_l_capsule, center=<-13.6327 -2.0199 0>, rotation=<rot 0 95.3936 90>}
  {
   shapeType=Capsule
   length=40.061
   radius=9.9436
   shapeName=calf_l_capsule
   center=<-31.9672 1.2815 0.0171>
   rotation=<rot -0.3859 86.1009 89.9295>
  }
 ]
 calf_r [
  {shapeType=Capsule, length=17.0018, radius=8.7708, shapeName=calf_r_capsule, center=<13.6327 2.0199 0>, rotation=<rot 0 95.3936 -90>}
  {
   shapeType=Capsule
   length=40.061
   radius=9.9436
   shapeName=calf_r_capsule
   center=<31.9672 -1.2815 -0.0171>
   rotation=<rot -0.3859 86.1009 -90.0705>
  }
 ]
 deform_skirt_back_l [{shapeType=Capsule, length=0.505, radius=0.505, shapeName=deform_skirt_back_l_capsule}]
 deform_skirt_back_r [{shapeType=Capsule, length=0.505, radius=0.505, shapeName=deform_skirt_back_r_capsule}]
 deform_skirt_side_l [{shapeType=Capsule, length=0.505, radius=0.505, shapeName=deform_skirt_side_l_capsule}]
 deform_skirt_side_r [{shapeType=Capsule, length=0.505, radius=0.505, shapeName=deform_skirt_side_r_capsule}]
 dyn_skirt_back_01_l [{
   shapeType=Capsule
   length=17.7171
   radius=9.6187
   shapeName=dyn_skirt_back_01_l_capsule
   center=<-8.9023 1.2283 5.2182>
   rotation=<rot 90 -90 0>
  }]
 dyn_skirt_back_01_r [{
   shapeType=Capsule
   length=17.7171
   radius=9.6193
   shapeName=dyn_skirt_back_01_r_capsule
   center=<8.9023 -1.2283 -5.2182>
   rotation=<rot 90 -90 0>
  }]
 dyn_skirt_back_02_l [{
   shapeType=Capsule
   length=24.6998
   radius=10.3597
   shapeName=dyn_skirt_back_02_l_capsule
   center=<-14.406 0.5393 4.3671>
   rotation=<rot 90 -90 0>
  }]
 dyn_skirt_back_02_r [{
   shapeType=Capsule
   length=24.6998
   radius=10.3597
   shapeName=dyn_skirt_back_02_r_capsule
   center=<14.406 -0.5393 -4.3671>
   rotation=<rot 90 90 0>
  }]
 dyn_skirt_back_base_l [{
   shapeType=Capsule
   length=11.4885
   radius=7.2902
   shapeName=dyn_skirt_back_base_l_capsule
   center=<-11.9509 -0.1089 2.4072>
   rotation=<rot 90 -90 0>
  }]
 dyn_skirt_back_base_r [{
   shapeType=Capsule
   length=11.4885
   radius=7.2902
   shapeName=dyn_skirt_back_base_r_capsule
   center=<11.9509 0.1089 -2.4072>
   rotation=<rot 90 90 0>
  }]
 dyn_skirt_side_01_l [{
   shapeType=Capsule
   length=13.8833
   radius=8.0687
   shapeName=dyn_skirt_side_01_l_capsule
   center=<-8.0604 -0.0746 -3.5795>
   rotation=<rot 90 -90 0>
  }]
 dyn_skirt_side_01_r [{shapeType=Capsule, length=15.7775, radius=7.6169, shapeName=dyn_skirt_side_01_r_capsule, center=<12 0 5.4237>, rotation=<rot 90 -90 0>}]
 dyn_skirt_side_base_l [{
   shapeType=Capsule
   length=9.8067
   radius=6.7582
   shapeName=dyn_skirt_side_base_l_capsule
   center=<-12.3783 -0.0016 -5.1237>
   rotation=<rot 90 -90 0>
  }]
 dyn_skirt_side_base_r [{
   shapeType=Capsule
   length=9.8067
   radius=6.7582
   shapeName=dyn_skirt_side_base_r_capsule
   center=<12.3639 0.0016 4.9279>
   rotation=<rot 90 90 0>
  }]
 foot_l [{shapeType=Capsule, length=1.5696, radius=14.4717, shapeName=foot_l_capsule, center=<-2.5446 4.4744 0.8404>, rotation=<rot -5.7371 89.9998 0>}]
 foot_r [{
   shapeType=Capsule
   length=1.5696
   radius=14.4717
   shapeName=foot_r_capsule
   center=<2.5446 -4.4744 -0.8404>
   rotation=<rot -5.7371 89.9998 180>
  }]
 pelvis [{shapeType=Capsule, length=8.5342, radius=15.4658, shapeName=spine_01_capsule}]
 spine_01 [{shapeType=Capsule, length=10.2653, radius=13.7347, shapeName=spine_01_capsule}]
 spine_02 [{shapeType=Capsule, length=12, radius=12, shapeName=spine_02_capsule}]
 spine_03 [{shapeType=Capsule, length=12, radius=12, shapeName=spine_03_capsule}]
 spine_04 [{shapeType=Capsule, length=12, radius=12, shapeName=spine_04_capsule}]
 spine_05 [{shapeType=Capsule, length=12, radius=12, shapeName=spine_05_capsule}]
 thigh_l [{
   shapeType=Capsule
   length=31.4458
   radius=6.0353
   shapeName=thigh_l_capsule
   center=<-28.3563 0.0002 0.0682>
   rotation=<rot 89.8621 0.0074 0.0078>
  }]
 thigh_r [{
   shapeType=Capsule
   length=29.642
   radius=7.8391
   shapeName=thigh_r_capsule
   center=<28.3543 -0.3309 -0.9077>
   rotation=<rot 89.8621 0.0074 -179.9922>
  }]
}
constraints [
 {bone1Name=dyn_skirt_back_01_r, bone2Name=dyn_skirt_back_base_r, swing1LimitDegrees=10, swing2LimitDegrees=40, twistLimitDegrees=3}
 {bone1Name=dyn_skirt_back_02_r, bone2Name=dyn_skirt_back_01_r, swing1LimitDegrees=15, swing2LimitDegrees=35, twistLimitDegrees=3}
 {bone1Name=dyn_skirt_back_02_l, bone2Name=dyn_skirt_back_01_l, swing1LimitDegrees=15, swing2LimitDegrees=35, twistLimitDegrees=3}
 {bone1Name=dyn_skirt_back_01_l, bone2Name=dyn_skirt_back_base_l, swing1LimitDegrees=10, swing2LimitDegrees=40, twistLimitDegrees=3}
 {bone1Name=dyn_skirt_side_01_l, bone2Name=dyn_skirt_side_base_l, swing1LimitDegrees=25, swing2LimitDegrees=25, twistLimitDegrees=3}
 {bone1Name=dyn_skirt_side_01_r, bone2Name=dyn_skirt_side_base_r, swing1LimitDegrees=25, swing2LimitDegrees=25, twistLimitDegrees=3}
 {bone1Name=dyn_skirt_back_base_l, bone2Name=deform_skirt_back_l, swing1LimitDegrees=10, swing2LimitDegrees=30, twistLimitDegrees=3}
 {bone1Name=dyn_skirt_back_base_r, bone2Name=deform_skirt_back_r, swing1LimitDegrees=10, swing2LimitDegrees=30, twistLimitDegrees=3}
 {bone1Name=dyn_skirt_side_base_l, bone2Name=deform_skirt_side_l, swing1LimitDegrees=2, swing2LimitDegrees=25, twistLimitDegrees=2}
 {bone1Name=dyn_skirt_side_base_r, bone2Name=deform_skirt_side_r, swing1LimitDegrees=1, swing2LimitDegrees=15, twistLimitDegrees=1}
 {
  bone1Name=deform_skirt_side_l
  bone2Name=thigh_l
  swing1LimitDegrees=45
  swing1Motion=Free
  swing2LimitDegrees=45
  swing2Motion=Free
  twistLimitDegrees=45
  twistMotion=Free
 }
 {
  bone1Name=deform_skirt_side_r
  bone2Name=thigh_r
  swing1LimitDegrees=45
  swing1Motion=Free
  swing2LimitDegrees=45
  swing2Motion=Free
  twistLimitDegrees=45
  twistMotion=Free
 }
 {
  bone1Name=thigh_l
  bone2Name=pelvis
  swing1LimitDegrees=45
  swing1Motion=Free
  swing2LimitDegrees=45
  swing2Motion=Free
  twistLimitDegrees=45
  twistMotion=Free
 }
 {
  bone1Name=thigh_r
  bone2Name=pelvis
  swing1LimitDegrees=45
  swing1Motion=Free
  swing2LimitDegrees=45
  swing2Motion=Free
  twistLimitDegrees=45
  twistMotion=Free
 }
 {
  bone1Name=deform_skirt_back_r
  bone2Name=pelvis
  swing1LimitDegrees=45
  swing1Motion=Free
  swing2LimitDegrees=45
  swing2Motion=Free
  twistLimitDegrees=45
  twistMotion=Free
 }
 {
  bone1Name=deform_skirt_back_l
  bone2Name=pelvis
  swing1LimitDegrees=45
  swing1Motion=Free
  swing2LimitDegrees=45
  swing2Motion=Free
  twistLimitDegrees=45
  twistMotion=Free
 }
 {
  bone1Name=calf_r
  bone2Name=thigh_r
  swing1LimitDegrees=45
  swing1Motion=Free
  swing2LimitDegrees=45
  swing2Motion=Free
  twistLimitDegrees=45
  twistMotion=Free
 }
 {
  bone1Name=calf_l
  bone2Name=thigh_l
  swing1LimitDegrees=45
  swing1Motion=Free
  swing2LimitDegrees=45
  swing2Motion=Free
  twistLimitDegrees=45
  twistMotion=Free
 }
 {
  bone1Name=foot_r
  bone2Name=calf_r
  swing1LimitDegrees=45
  swing1Motion=Free
  swing2LimitDegrees=45
  swing2Motion=Free
  twistLimitDegrees=45
  twistMotion=Free
 }
 {
  bone1Name=foot_l
  bone2Name=calf_l
  swing1LimitDegrees=45
  swing1Motion=Free
  swing2LimitDegrees=45
  swing2Motion=Free
  twistLimitDegrees=45
  twistMotion=Free
 }
 {
  bone1Name=dyn_skirt_back_02_r
  bone2Name=dyn_skirt_back_02_l
  swing1LimitDegrees=15
  swing2LimitDegrees=45
  swing2Motion=Free
  twistLimitDegrees=45
  twistMotion=Free
 }
 {
  bone1Name=thigh_r
  bone2Name=dyn_skirt_back_02_r
  swing1LimitDegrees=25
  swing1Motion=Free
  swing2LimitDegrees=25
  swing2Motion=Free
  twistLimitDegrees=25
  twistMotion=Free
 }
 {
  bone1Name=thigh_l
  bone2Name=dyn_skirt_back_02_l
  swing1LimitDegrees=45
  swing1Motion=Free
  swing2LimitDegrees=25
  swing2Motion=Free
  twistLimitDegrees=45
  twistMotion=Free
 }
 {
  bone1Name=spine_01
  bone2Name=pelvis
  swing1LimitDegrees=45
  swing1Motion=Free
  swing2LimitDegrees=45
  swing2Motion=Free
  twistLimitDegrees=45
  twistMotion=Free
 }
 {
  bone1Name=spine_02
  bone2Name=spine_01
  swing1LimitDegrees=45
  swing1Motion=Free
  swing2LimitDegrees=45
  swing2Motion=Free
  twistLimitDegrees=45
  twistMotion=Free
 }
 {
  bone1Name=spine_03
  bone2Name=spine_02
  swing1LimitDegrees=45
  swing1Motion=Free
  swing2LimitDegrees=45
  swing2Motion=Free
  twistLimitDegrees=45
  twistMotion=Free
 }
 {
  bone1Name=spine_04
  bone2Name=spine_03
  swing1LimitDegrees=45
  swing1Motion=Free
  swing2LimitDegrees=45
  swing2Motion=Free
  twistLimitDegrees=45
  twistMotion=Free
 }
 {
  bone1Name=spine_05
  bone2Name=spine_04
  swing1LimitDegrees=45
  swing1Motion=Free
  swing2LimitDegrees=45
  swing2Motion=Free
  twistLimitDegrees=45
  twistMotion=Free
 }
]
