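# Physics asset for the F_MED_PrairieGizmo character's DBNO (downed) variant, built around skirt dynamics.
# Capsules cover pelvis, legs, feet and seven skirt chains, plus root shapes including a huge one below the origin, likely a floor stand-in.
# Skirt chains hang from MSTR bones on the pelvis, except left front and side on thigh_l, with mostly limited joints.
# Angular-free constraints cross-link some neighbouring chains to hold panels together.
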PhysicsAsset /BRCosmetics/Characters/Player/Female/Medium/Bodies/F_MED_PrairieGizmo/Meshes/F_MED_PrairieGizmo_Physics_DBNO
bodies {
 calf_l [{shapeType=Capsule, length=45, radius=7, center=<-27.1965 -2.2262 0>, rotation=<rot 90 0 0>}]
 calf_r [{shapeType=Capsule, length=30, radius=6, center=<28.6459 0.6978 -0.1562>, rotation=<rot 90 17.354 23.0986>}]
 dyn_skirt_bk_C_01 [{shapeType=Capsule, length=15, radius=2, shapeName=dyn_skirt_bk_C_01_capsule, rotation=<rot 0 0 90>}]
 dyn_skirt_bk_C_02 [{shapeType=Capsule, length=10, radius=3, shapeName=dyn_skirt_bk_C_02_capsule, center=<1.546 0 -2.1154>, rotation=<rot 0 0 90>}]
 dyn_skirt_bk_C_03 [{shapeType=Capsule, length=15, radius=3, shapeName=dyn_skirt_bk_C_03_capsule, center=<1.7715 0.0688 -1.6109>, rotation=<rot 0.0001 0 93.5073>}]
 dyn_skirt_bk_C_04 [{
   shapeType=Capsule
   length=20
   radius=3
   shapeName=dyn_skirt_bk_C_04_capsule
   center=<3.8268 -1.0369 -0.5781>
   rotation=<rot 0.6306 -0.1098 -86.9674>
  }]
 dyn_skirt_bk_C_MSTR [{shapeType=Capsule, length=0.505, radius=1, shapeName=dyn_skirt_bk_C_MSTR_capsule, rotation=<rot 0 0 90>}]
 dyn_skirt_bk_l_01 [{shapeType=Capsule, length=6, radius=2, shapeName=dyn_skirt_bk_l_01_capsule, center=<4.7361 0 -1.5633>, rotation=<rot 0 0 90>}]
 dyn_skirt_bk_l_02 [{shapeType=Capsule, length=15, radius=3, shapeName=dyn_skirt_bk_l_02_capsule, center=<1.9442 0 -1.2706>, rotation=<rot 0 0 90>}]
 dyn_skirt_bk_l_03 [{
   shapeType=Capsule
   length=15
   radius=3
   shapeName=dyn_skirt_bk_l_03_capsule
   center=<5.6432 -3.4054 -1.1568>
   rotation=<rot 0.0034 141.2866 -89.9172>
  }]
 dyn_skirt_bk_l_MSTR [{shapeType=Capsule, length=0.505, radius=0.5, shapeName=dyn_skirt_bk_l_MSTR_capsule, rotation=<rot 0 0 90>}]
 dyn_skirt_bk_r_01 [{
   shapeType=Capsule
   length=12
   radius=2
   shapeName=dyn_skirt_bk_r_01_capsule
   center=<2.831 -0.2908 -0.4561>
   rotation=<rot 0.0002 0.0001 94.527>
  }]
 dyn_skirt_bk_r_02 [{shapeType=Capsule, length=12, radius=3, shapeName=dyn_skirt_bk_r_02_capsule, center=<5.7366 1.6409 -1.6051>, rotation=<rot 0.0001 0.73 90>}]
 dyn_skirt_bk_r_03 [{
   shapeType=Capsule
   length=17
   radius=3
   shapeName=dyn_skirt_bk_r_03_capsule
   center=<12.0827 -5.1861 -0.9788>
   rotation=<rot -10.5459 38.0534 107.6328>
  }]
 dyn_skirt_bk_r_MSTR [{shapeType=Capsule, length=0.505, radius=0.505, shapeName=dyn_skirt_bk_r_MSTR_capsule, rotation=<rot 0 -90 0>}]
 dyn_skirt_fr_l_01 [{shapeType=Capsule, length=5, radius=1, shapeName=dyn_skirt_fr_l_01_capsule, center=<3.555 0 -0.9082>, rotation=<rot 0 -30.1421 90>}]
 dyn_skirt_fr_l_02 [{
   shapeType=Capsule
   length=7.8822
   radius=3
   shapeName=dyn_skirt_fr_l_02_capsule
   center=<4.9884 -0.195 -2>
   rotation=<rot -0.0001 -43.7446 90>
  }]
 dyn_skirt_fr_l_MSTR [{shapeType=Capsule, length=0, radius=1, shapeName=dyn_skirt_fr_l_MSTR_capsule, rotation=<rot 0 0 90>}]
 dyn_skirt_fr_r_01 [{
   shapeType=Capsule
   length=4
   radius=1
   shapeName=dyn_skirt_fr_r_01_capsule
   center=<3.1814 -5.2145 2.0182>
   rotation=<rot 22.3554 -87.1362 93.7561>
  }]
 dyn_skirt_fr_r_02 [{
   shapeType=Capsule
   length=6.9859
   radius=2
   shapeName=dyn_skirt_fr_r_02_capsule
   center=<3.3429 -1.0155 -1.4467>
   rotation=<rot 15.1026 -99.4412 85.9521>
  }]
 dyn_skirt_fr_r_MSTR [{shapeType=Capsule, length=0.505, radius=1, shapeName=dyn_skirt_fr_r_MSTR_capsule, rotation=<rot 0 0 90>}]
 dyn_skirt_side_l_01 [{shapeType=Capsule, length=8.4734, radius=2, shapeName=dyn_skirt_side_l_01_capsule, center=<2.6211 0 -3.2187>, rotation=<rot 0 0 90>}]
 dyn_skirt_side_l_02 [{shapeType=Capsule, length=8.5768, radius=3, shapeName=dyn_skirt_side_l_02_capsule, center=<5.1832 0 -2>, rotation=<rot 0 0 83.9011>}]
 dyn_skirt_side_l_MSTR [{shapeType=Capsule, length=0.505, radius=1, shapeName=dyn_skirt_side_l_MSTR_capsule, rotation=<rot 0 0 90>}]
 dyn_skirt_side_r_01 [{
   shapeType=Capsule
   length=10
   radius=3
   shapeName=dyn_skirt_side_r_01_capsule
   center=<3.7664 3.5057 -1.846>
   rotation=<rot -0.0001 0.0005 54.9044>
  }]
 dyn_skirt_side_r_02 [{
   shapeType=Capsule
   length=12
   radius=3
   shapeName=dyn_skirt_side_r_02_capsule
   center=<1.0847 1.5786 -0.7013>
   rotation=<rot 2.2065 -5.6708 76.2123>
  }]
 dyn_skirt_side_r_03 [{
   shapeType=Capsule
   length=15
   radius=6
   shapeName=dyn_skirt_side_r_03_capsule
   center=<4.798 -0.0533 -4.4714>
   rotation=<rot -18.3281 38.7493 94.2274>
  }]
 dyn_skirt_side_r_MSTR [{shapeType=Capsule, length=0, radius=0.505, shapeName=dyn_skirt_side_r_MSTR_capsule, rotation=<rot 0 90 0>}]
 foot_l [{shapeType=Capsule, length=8.6658, radius=8, shapeName=foot_l_capsule, center=<-3.6944 3.9746 0.1494>, rotation=<rot -0.7858 8.1221 -86.7528>}]
 foot_r [{
   shapeType=Capsule
   length=8.6688
   radius=8
   shapeName=foot_r_capsule
   center=<4.772 -3.0904 -0.0663>
   rotation=<rot -0.9532 10.6773 -86.7296>
  }]
 pelvis [{shapeType=Capsule, length=10, radius=10, center=<0 -1.0608 0>}]
 root [
  {shapeType=Capsule, length=100, radius=200, shapeName=root_capsule, center=<0 -31.4676 -199.858>, rotation=<rot 86.516 -90 0>}
  {shapeType=Capsule, length=15, radius=20, shapeName=root_capsule, center=<-7.0672 -11.3229 4.6583>, rotation=<rot 0 0 22.3628>}
 ]
 thigh_l [{shapeType=Capsule, length=38, radius=8, center=<-17.7562 0 -0.0174>, rotation=<rot 89.7809 179.9942 179.9945>}]
 thigh_r [{shapeType=Capsule, length=45, radius=8.5, center=<18.2411 1.7952 -0.4973>, rotation=<rot 88.9395 -5.141 0.0003>}]
}
constraints [
 {bone1Name=thigh_l, bone2Name=pelvis, swing1LimitDegrees=45, swing2LimitDegrees=45, twistLimitDegrees=45}
 {bone1Name=calf_l, bone2Name=thigh_l, swing1LimitDegrees=45, swing2LimitDegrees=45, twistLimitDegrees=45}
 {bone1Name=thigh_r, bone2Name=pelvis, swing1LimitDegrees=45, swing2LimitDegrees=45, twistLimitDegrees=45}
 {bone1Name=calf_r, bone2Name=thigh_r, swing1LimitDegrees=45, swing2LimitDegrees=45, twistLimitDegrees=45}
 {bone1Name=dyn_skirt_side_r_03, bone2Name=dyn_skirt_side_r_02, swing1LimitDegrees=10, swing2LimitDegrees=25, twistLimitDegrees=5}
 {
  bone1Name=dyn_skirt_side_r_MSTR
  bone2Name=pelvis
  swing1LimitDegrees=45
  swing1Motion=Free
  swing2LimitDegrees=45
  swing2Motion=Free
  twistLimitDegrees=45
  twistMotion=Free
 }
 {bone1Name=dyn_skirt_side_r_01, bone2Name=dyn_skirt_side_r_MSTR, swing1LimitDegrees=25, swing2LimitDegrees=25, twistLimitDegrees=10}
 {bone1Name=dyn_skirt_side_r_02, bone2Name=dyn_skirt_side_r_01, swing1LimitDegrees=35, swing2LimitDegrees=35, twistLimitDegrees=20}
 {bone1Name=dyn_skirt_bk_r_03, bone2Name=dyn_skirt_bk_r_02, swing1LimitDegrees=25, swing2LimitDegrees=30, twistLimitDegrees=15}
 {bone1Name=dyn_skirt_bk_r_02, bone2Name=dyn_skirt_bk_r_01, swing1LimitDegrees=30, swing2LimitDegrees=40, twistLimitDegrees=10}
 {bone1Name=dyn_skirt_bk_r_01, bone2Name=dyn_skirt_bk_r_MSTR, swing1LimitDegrees=20, swing2LimitDegrees=20, twistLimitDegrees=10}
 {
  bone1Name=dyn_skirt_bk_r_MSTR
  bone2Name=pelvis
  swing1LimitDegrees=45
  swing1Motion=Free
  swing2LimitDegrees=45
  swing2Motion=Free
  twistLimitDegrees=45
  twistMotion=Free
 }
 {bone1Name=dyn_skirt_fr_l_02, bone2Name=dyn_skirt_fr_l_01, swing1LimitDegrees=25, swing2LimitDegrees=25, twistLimitDegrees=10}
 {bone1Name=dyn_skirt_side_l_02, bone2Name=dyn_skirt_side_l_01, swing1LimitDegrees=25, swing2LimitDegrees=25, twistLimitDegrees=10}
 {
  bone1Name=dyn_skirt_fr_l_02
  bone2Name=dyn_skirt_side_l_02
  swing1LimitDegrees=25
  swing1Motion=Free
  swing2LimitDegrees=25
  swing2Motion=Free
  twistLimitDegrees=10
  twistMotion=Free
 }
 {bone1Name=dyn_skirt_fr_r_02, bone2Name=dyn_skirt_fr_r_01, swing1LimitDegrees=25, swing2LimitDegrees=25, twistLimitDegrees=10}
 {
  bone1Name=dyn_skirt_fr_l_02
  bone2Name=dyn_skirt_fr_r_02
  swing1LimitDegrees=25
  swing1Motion=Free
  swing2LimitDegrees=25
  swing2Motion=Free
  twistLimitDegrees=10
  twistMotion=Free
 }
 {bone1Name=dyn_skirt_fr_r_01, bone2Name=dyn_skirt_fr_r_MSTR, swing1LimitDegrees=10, swing2LimitDegrees=25, twistLimitDegrees=5}
 {bone1Name=dyn_skirt_fr_r_MSTR, bone2Name=pelvis, swing1LimitDegrees=25, swing2LimitDegrees=25, twistLimitDegrees=10}
 {bone1Name=dyn_skirt_bk_l_02, bone2Name=dyn_skirt_bk_l_01, swing1LimitDegrees=30, swing2LimitDegrees=45, twistLimitDegrees=20}
 {bone1Name=dyn_skirt_bk_l_01, bone2Name=dyn_skirt_bk_l_MSTR, swing1LimitDegrees=10, swing2LimitDegrees=35, twistLimitDegrees=5}
 {
  bone1Name=dyn_skirt_bk_l_MSTR
  bone2Name=pelvis
  swing1LimitDegrees=45
  swing1Motion=Free
  swing2LimitDegrees=45
  swing2Motion=Free
  twistLimitDegrees=45
  twistMotion=Free
 }
 {bone1Name=dyn_skirt_bk_C_03, bone2Name=dyn_skirt_bk_C_02, swing1LimitDegrees=20, swing2LimitDegrees=60, twistLimitDegrees=10}
 {bone1Name=dyn_skirt_bk_C_02, bone2Name=dyn_skirt_bk_C_01, swing1LimitDegrees=20, swing2LimitDegrees=50, twistLimitDegrees=10}
 {
  bone1Name=dyn_skirt_bk_C_01
  bone2Name=dyn_skirt_bk_C_MSTR
  swing1LimitDegrees=25
  swing1Motion=Free
  swing2LimitDegrees=25
  swing2Motion=Free
  twistLimitDegrees=10
  twistMotion=Free
 }
 {bone1Name=dyn_skirt_bk_C_MSTR, bone2Name=pelvis, swing1LimitDegrees=25, swing2LimitDegrees=25, twistLimitDegrees=10}
 {bone1Name=dyn_skirt_side_l_01, bone2Name=dyn_skirt_side_l_MSTR, swing1LimitDegrees=25, swing2LimitDegrees=25, twistLimitDegrees=10}
 {
  bone1Name=dyn_skirt_side_l_MSTR
  bone2Name=thigh_l
  swing1LimitDegrees=45
  swing1Motion=Free
  swing2LimitDegrees=45
  swing2Motion=Free
  twistLimitDegrees=45
  twistMotion=Free
 }
 {bone1Name=dyn_skirt_fr_l_01, bone2Name=dyn_skirt_fr_l_MSTR, swing1LimitDegrees=25, swing2LimitDegrees=25, twistLimitDegrees=10}
 {bone1Name=dyn_skirt_fr_l_MSTR, bone2Name=thigh_l, swing1LimitDegrees=25, swing2LimitDegrees=25, twistLimitDegrees=10}
 {
  bone1Name=dyn_skirt_side_l_01
  bone2Name=dyn_skirt_fr_l_01
  swing1LimitDegrees=25
  swing1Motion=Free
  swing2LimitDegrees=25
  swing2Motion=Free
  twistLimitDegrees=10
  twistMotion=Free
 }
 {
  bone1Name=pelvis
  bone2Name=root
  swing1LimitDegrees=45
  swing1Motion=Free
  swing2LimitDegrees=45
  swing2Motion=Free
  twistLimitDegrees=45
  twistMotion=Free
 }
 {
  bone1Name=foot_r
  bone2Name=calf_r
  swing1LimitDegrees=45
  swing1Motion=Free
  swing2LimitDegrees=45
  swing2Motion=Free
  twistLimitDegrees=45
  twistMotion=Free
 }
 {
  bone1Name=foot_l
  bone2Name=calf_l
  swing1LimitDegrees=45
  swing1Motion=Free
  swing2LimitDegrees=45
  swing2Motion=Free
  twistLimitDegrees=45
  twistMotion=Free
 }
 {bone1Name=dyn_skirt_bk_C_04, bone2Name=dyn_skirt_bk_C_03, swing1LimitDegrees=15, swing2LimitDegrees=50, twistLimitDegrees=10}
 {bone1Name=dyn_skirt_bk_l_03, bone2Name=dyn_skirt_bk_l_02, swing1LimitDegrees=15, swing2LimitDegrees=30, twistLimitDegrees=20}
 {
  bone1Name=dyn_skirt_bk_C_04
  bone2Name=dyn_skirt_bk_r_03
  swing1LimitDegrees=45
  swing1Motion=Free
  swing2LimitDegrees=45
  swing2Motion=Free
  twistLimitDegrees=45
  twistMotion=Free
 }
 {
  bone1Name=dyn_skirt_side_r_02
  bone2Name=dyn_skirt_bk_r_01
  swing1LimitDegrees=45
  swing1Motion=Free
  swing2LimitDegrees=45
  swing2Motion=Free
  twistLimitDegrees=45
  twistMotion=Free
 }
]
