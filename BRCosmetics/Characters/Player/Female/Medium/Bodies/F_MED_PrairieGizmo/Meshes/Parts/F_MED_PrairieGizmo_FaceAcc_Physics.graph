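# Physics asset for the PrairieGizmo female medium face accessory, giving collision and joint limits to mirrored tail and hair dynamic bone chains.
# A head sphere and capsule accompany capsules on each chain's master and segment bones.
# Three-segment tails and two-segment hair strands use narrow swing limits and little or no twist, with slightly asymmetric outer tail limits.
# Tail masters join the head with free rotation; hair masters have no head constraint.
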PhysicsAsset /BRCosmetics/Characters/Player/Female/Medium/Bodies/F_MED_PrairieGizmo/Meshes/Parts/F_MED_PrairieGizmo_FaceAcc_Physics
bodies {
 dyn_hair_locan_01_l [{
   shapeType=Capsule
   length=2.5
   radius=1.4494
   shapeName=dyn_hair_locan_01_l_capsule
   center=<1.5562 -0.0708 -0.2>
   rotation=<rot 84.5985 -2.6064 0>
  }]
 dyn_hair_locan_01_r [{
   shapeType=Capsule
   length=2.5
   radius=1.4494
   shapeName=dyn_hair_locan_01_r_capsule
   center=<1.5481 -0.0705 -0.4551>
   rotation=<rot 84.8486 -2.6065 -0.0003>
  }]
 dyn_hair_locan_02_l [{
   shapeType=Capsule
   length=6.3042
   radius=1.2241
   shapeName=dyn_hair_locan_02_l_capsule
   center=<3.2765 -0.7719 0.4786>
   rotation=<rot 78.6802 179.9992 -170.0135>
  }]
 dyn_hair_locan_02_r [{
   shapeType=Capsule
   length=6.3042
   radius=1.2241
   shapeName=dyn_hair_locan_02_r_capsule
   center=<3.4299 0.4767 0.0003>
   rotation=<rot 83.0306 -179.9981 172.8601>
  }]
 dyn_hair_locan_MSTR_l [{shapeType=Capsule, length=0.505, radius=0.505, shapeName=dyn_hair_locan_MSTR_l_capsule, rotation=<rot 0 -90 0>}]
 dyn_hair_locan_MSTR_r [{shapeType=Capsule, length=0.505, radius=0.505, shapeName=dyn_hair_locan_MSTR_r_capsule, rotation=<rot 0 -90 0>}]
 dyn_tail_01_l [{shapeType=Capsule, length=3, radius=3, center=<3.3819 0 -0.9492>, rotation=<rot 90 0 0>}]
 dyn_tail_01_r [{shapeType=Capsule, length=3, radius=3, center=<3.3819 0 -0.9492>, rotation=<rot 90 0 0>}]
 dyn_tail_02_l [{shapeType=Capsule, length=2, radius=3, shapeName=dyn_tail_02_l_capsule, center=<3.3317 0 0>, rotation=<rot 90 180 -180>}]
 dyn_tail_02_r [{shapeType=Capsule, length=2, radius=3, shapeName=dyn_tail_02_r_capsule, center=<3.3317 0 0>, rotation=<rot 90 180 180>}]
 dyn_tail_03_l [{
   shapeType=Capsule
   length=3.1806
   radius=2.2096
   shapeName=dyn_tail_03_l_capsule
   center=<4.3813 -0.2026 -0.9701>
   rotation=<rot 0 -87.2175 -45.0913>
  }]
 dyn_tail_03_r [{
   shapeType=Capsule
   length=3.2908
   radius=2.2923
   shapeName=dyn_tail_03_r_capsule
   center=<4.2088 0.0584 -0.9973>
   rotation=<rot 7.0683 -90 -43.8408>
  }]
 dyn_tail_MSTR_l [{shapeType=Capsule, length=0, radius=0.505, shapeName=dyn_tail_MSTR_l_capsule, rotation=<rot 0 -90 0>}]
 dyn_tail_MSTR_r [{shapeType=Capsule, length=0, radius=0.505, shapeName=dyn_tail_MSTR_r_capsule, rotation=<rot 0 -90 0>}]
 head [
  {shapeType=Sphere, radius=10, shapeName=head_sphere, center=<5.7608 -1.2395 0>}
  {shapeType=Capsule, length=7.4688, radius=5.4375, center=<0.5445 3.4826 0>, rotation=<rot 90 -0.0001 0>}
 ]
}
constraints [
 {bone1Name=dyn_tail_01_l, bone2Name=dyn_tail_MSTR_l, swing1LimitDegrees=6, swing2LimitDegrees=20, twistLimitDegrees=0}
 {bone1Name=dyn_tail_01_r, bone2Name=dyn_tail_MSTR_r, swing1LimitDegrees=6, swing2LimitDegrees=20, twistLimitDegrees=0}
 {bone1Name=dyn_hair_locan_01_l, bone2Name=dyn_hair_locan_MSTR_l, swing1LimitDegrees=15, swing2LimitDegrees=10, twistLimitDegrees=0}
 {bone1Name=dyn_hair_locan_02_l, bone2Name=dyn_hair_locan_01_l, swing1LimitDegrees=15, swing2LimitDegrees=15, twistLimitDegrees=0}
 {bone1Name=dyn_hair_locan_02_r, bone2Name=dyn_hair_locan_01_r, swing1LimitDegrees=15, swing2LimitDegrees=15, twistLimitDegrees=0}
 {bone1Name=dyn_hair_locan_01_r, bone2Name=dyn_hair_locan_MSTR_r, swing1LimitDegrees=15, swing2LimitDegrees=10, twistLimitDegrees=0}
 {bone1Name=dyn_tail_03_r, bone2Name=dyn_tail_02_r, swing1LimitDegrees=30, swing2LimitDegrees=15, twistLimitDegrees=2}
 {bone1Name=dyn_tail_02_r, bone2Name=dyn_tail_01_r, swing1LimitDegrees=20, swing2LimitDegrees=30, twistLimitDegrees=2}
 {
  bone1Name=dyn_tail_MSTR_r
  bone2Name=head
  swing1LimitDegrees=45
  swing1Motion=Free
  swing2LimitDegrees=45
  swing2Motion=Free
  twistLimitDegrees=45
  twistMotion=Free
 }
 {bone1Name=dyn_tail_03_l, bone2Name=dyn_tail_02_l, swing1LimitDegrees=20, swing2LimitDegrees=15, twistLimitDegrees=2}
 {bone1Name=dyn_tail_02_l, bone2Name=dyn_tail_01_l, swing1LimitDegrees=15, swing2LimitDegrees=30, twistLimitDegrees=2}
 {
  bone1Name=dyn_tail_MSTR_l
  bone2Name=head
  swing1LimitDegrees=45
  swing1Motion=Free
  swing2LimitDegrees=45
  swing2Motion=Free
  twistLimitDegrees=45
  twistMotion=Free
 }
]
